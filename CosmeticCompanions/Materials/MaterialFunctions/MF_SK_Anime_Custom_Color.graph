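# Material function that builds a customizable base color for anime-style skeletal mesh materials in CosmeticCompanions.
# Static switches optionally blend each custom base, primary, secondary and tertiary color with a second color for advanced blending.
# Chained mask lerps layer these colors, with an optional strength-scaled pattern overlay.
# The result is multiplied, then lerped over another input, either fully or through a mask intended to preserve original diffuse areas.
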MaterialFunction /CosmeticCompanions/Materials/MaterialFunctions/MF_SK_Anime_Custom_Color
graph {nodes [
  FunctionInput() {inputName=----------------------, inputType=FunctionInput_Scalar, sortPriority=-1}
  FunctionOutput(None=NamedRerouteUsage()) {outputName=Color}
  NamedRerouteDeclaration(None=FunctionInput() {inputName=Primary Mask, inputType=FunctionInput_Scalar, sortPriority=-29}) {name=PrimaryMask, nodeColor=<rgba 1 0.0596 0.1687 1>}
  NamedRerouteDeclaration(None=FunctionInput() {inputName=Secondary Mask, inputType=FunctionInput_Scalar, sortPriority=-28}) {name=SecondaryMask, nodeColor=<rgba 0 1 0.0283 1>}
  NamedRerouteDeclaration(None=FunctionInput() {bUsePreviewValueAsDefault=true, inputName=Advanced Blend Mask, inputType=FunctionInput_Scalar, sortPriority=-26}) {name=AdvancedBlendMask, nodeColor=<rgba 0.5703 1 0 1>}
  NamedRerouteDeclaration(None=FunctionInput() {bUsePreviewValueAsDefault=true, inputName=Diffuse Base Color Value, inputType=FunctionInput_Scalar, sortPriority=1}) {name=DiffuseBaseColorValue, nodeColor=<rgba 0.1618 0.2425 1 1>}
  n1 = NamedRerouteUsage()
  n2 = NamedRerouteUsage()
  n3 = NamedRerouteUsage()
  NamedRerouteDeclaration(None=StaticSwitch(True=LinearInterpolate(A=@n1, B=@n2, Alpha=@n3), False=@n1, Value=NamedRerouteUsage())) {name=PrimaryColor, nodeColor=<rgba 0 1 0.7891 1>}
  n4 = NamedRerouteUsage()
  n5 = NamedRerouteUsage()
  n6 = NamedRerouteUsage()
  NamedRerouteDeclaration(None=StaticSwitch(True=LinearInterpolate(A=@n4, B=@n5, Alpha=@n6), False=@n4, Value=NamedRerouteUsage())) {name=SecondaryColor, nodeColor=<rgba 0 1 0.7891 1>}
  n7 = NamedRerouteUsage()
  n8 = NamedRerouteUsage()
  n9 = NamedRerouteUsage()
  NamedRerouteDeclaration(None=StaticSwitch(True=LinearInterpolate(A=@n7, B=@n8, Alpha=@n9), False=@n7, Value=NamedRerouteUsage())) {name=TertiaryColor, nodeColor=<rgba 0 1 0.7891 1>}
  NamedRerouteDeclaration(None=FunctionInput() {inputName=Pattern Mask, sortPriority=-26}) {name=PatternMask, nodeColor=<rgba 0.1719 1 0 1>}
  n10 = NamedRerouteUsage()
  n11 = NamedRerouteUsage()
  n12 = NamedRerouteUsage()
  n13 = NamedRerouteUsage()
  n14 = LinearInterpolate(
   A=LinearInterpolate(A=LinearInterpolate(A=@n11, B=@n12, Alpha=@n13) {constA=1}, B=NamedRerouteUsage(), Alpha=NamedRerouteUsage())
   B=NamedRerouteUsage()
   Alpha=NamedRerouteUsage()
  )
  n15 = NamedRerouteUsage()
  n16 = ScalarParameter() {group=Color Customization, parameterName=Global_PatternStrength, sortPriority=54}
  n17 = StaticSwitchParameter(True=LinearInterpolate(A=@n14, B=NamedRerouteUsage(), Alpha=Multiply(A=@n15, B=@n16)) {constA=1}, False=@n14) {group=Color Customization, parameterName=Use Pattern Mask, sortPriority=50}
  n18 = StaticSwitchParameter(True=NamedRerouteUsage(), False=Constant() {r=1}) {
   defaultValue=true
   desc="Use this to mask out areas on the base/diffuse if we want to keep the diffuse colors in areas "
   group=Color Customization
   parameterName=Use FX Mask For Masking Base
   sortPriority=-11
  }
  NamedRerouteDeclaration(None=LinearInterpolate(A=NamedRerouteUsage(), B=Multiply(A=@n10, B=@n17), Alpha=@n18)) {name=OUT_BaseColor, nodeColor=<rgba 1 0 0.4922 1>}
  NamedRerouteDeclaration(None=FunctionInput() {bUsePreviewValueAsDefault=true, inputName=Tech Art Mask, inputType=FunctionInput_Scalar, sortPriority=-30}) {name=BaseMask, nodeColor=<rgba 0 0.8125 1 1>}
  NamedRerouteDeclaration(None=FunctionInput() {inputName=Tertiary Mask, inputType=FunctionInput_Scalar, sortPriority=-27}) {name=TertiaryMask, nodeColor=<rgba 0 0.2243 1 1>}
  NamedRerouteDeclaration(None=FunctionInput() {inputName=Custom Base Color, sortPriority=2}) {name=IN_CustomBaseColor, nodeColor=<rgba 0.7194 0.1895 1 1>}
  NamedRerouteDeclaration(None=FunctionInput() {inputName=Custom Base Color Blend, sortPriority=3}) {name=IN_CustomBaseColorBlend, nodeColor=<rgba 0.7194 0.1895 1 1>}
  NamedRerouteDeclaration(None=FunctionInput() {inputName=Primary Color, sortPriority=4}) {name=IN_PrimaryColor, nodeColor=<rgba 0.3093 1 0.4661 1>}
  NamedRerouteDeclaration(None=FunctionInput() {inputName=Primary Color Blend, sortPriority=5}) {name=IN_PrimaryColorBlend, nodeColor=<rgba 0.3093 1 0.4661 1>}
  NamedRerouteDeclaration(None=FunctionInput() {inputName=Secondary Color, sortPriority=6}) {name=IN_SecondaryColor, nodeColor=<rgba 0.257 0 1 1>}
  NamedRerouteDeclaration(None=FunctionInput() {inputName=Secondary Color Blend, sortPriority=7}) {name=IN_SecondaryColorBlend, nodeColor=<rgba 0.257 0 1 1>}
  NamedRerouteDeclaration(None=FunctionInput() {inputName=Tertiary Color, sortPriority=8}) {name=IN_TertiaryColor, nodeColor=<rgba 1 0.7304 0.2272 1>}
  NamedRerouteDeclaration(None=FunctionInput() {inputName=Tertiary Blend Color, sortPriority=9}) {name=IN_TertiaryBlendColor, nodeColor=<rgba 1 0.7304 0.2272 1>}
  NamedRerouteDeclaration(None=FunctionInput() {inputName=Pattern Color, sortPriority=10}) {name=IN_PatternColor, nodeColor=<rgba 0.0171 0.8364 1 1>}
  NamedRerouteDeclaration(None=FunctionInput() {bUsePreviewValueAsDefault=true, inputName=Use Advanced Color Blend, inputType=FunctionInput_StaticBool, sortPriority=-25}) {name=AdvancedColorBlend, nodeColor=<rgba 0.7109 0 1 1>}
  n19 = NamedRerouteUsage()
  n20 = NamedRerouteUsage()
  n21 = NamedRerouteUsage()
  NamedRerouteDeclaration(None=StaticSwitch(True=LinearInterpolate(A=@n19, B=@n20, Alpha=@n21), False=@n19, Value=NamedRerouteUsage())) {name=CustomBaseColor, nodeColor=<rgba 0 1 0.3906 1>}
  NamedRerouteDeclaration(None=FunctionInput() {inputName=Diffuse Base Color}) {name=IN_DiffuseBaseColor, nodeColor=<rgba 0.1618 0.2425 1 1>}
 ]}
