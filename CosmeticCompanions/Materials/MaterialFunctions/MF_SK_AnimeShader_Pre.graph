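# Material function that prepares inputs for the Cosmetic Companions anime shader: lit/shaded colors, normal, DFL, STT, SRM, specular, emissive, FX and rim color.
# Static switches choose maps or constant parameters, with optional darkened colors derived for shaded and DFL.
# Mask-driven MF_SK_Anime_Custom_Color calls can replace lit and shaded colors before cinematic tint functions run.
# STT packs sheen, threshold and thickness; emissive is scaled by a view-normal Fresnel-style term.
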MaterialFunction /CosmeticCompanions/Materials/MaterialFunctions/MF_SK_AnimeShader_Pre
graph {nodes [
  FunctionOutput(None=NamedRerouteUsage()) {outputName=Color_Lit, sortPriority=11}
  FunctionOutput(None=NamedRerouteUsage()) {outputName=Color_Shaded, sortPriority=12}
  n1 = VectorParameter() {desc="RGB=Color", group=01_Anime Maps, parameterName=StyleRimLight_Color, sortPriority=22, defaultValue=<rgba 1 1 1 1>}
  FunctionOutput(None=@n1.RGB) {outputName=StyleRimLight_Color, sortPriority=17}
  FunctionOutput(None=StaticSwitchParameter(True=NamedRerouteUsage(), False=Constant()) {group=01_Anime Maps, parameterName=UseDFLMap, sortPriority=4}) {outputName=DFL, sortPriority=3}
  FunctionOutput(None=StaticSwitchParameter(True=NamedRerouteUsage(), False=Constant3Vector() {constant=<rgba 0 0 1 0>}) {group=01_Anime Maps, parameterName=UseNormalMap, sortPriority=3}) {outputName=Normal, sortPriority=2}
  FunctionOutput(None=NamedRerouteUsage()) {outputName=STT, sortPriority=1}
  FunctionOutput(None=NamedRerouteUsage()) {outputName=DFL_Color, sortPriority=15}
  FunctionOutput(None=NamedRerouteUsage()) {outputName=Emissive, sortPriority=14}
  FunctionOutput(None=NamedRerouteUsage()) {outputName=SRM, sortPriority=13}
  FunctionOutput(None=NamedRerouteUsage()) {outputName=FX, sortPriority=4}
  FunctionOutput() {outputName=--------, sortPriority=10}
  FunctionOutput(None=NamedRerouteUsage()) {outputName=Specular_Color, sortPriority=16}
  n2 = MaterialFunctionCall!(
   Tech Art Mask=NamedRerouteUsage()
   Primary Mask=NamedRerouteUsage()
   Secondary Mask=NamedRerouteUsage()
   Tertiary Mask=NamedRerouteUsage()
   Pattern Mask=NamedRerouteUsage()
   Advanced Blend Mask=NamedRerouteUsage()
   Use Advanced Color Blend=NamedRerouteUsage()
   ----------------------
   Diffuse Base Color=NamedRerouteUsage()
   Diffuse Base Color Value=NamedRerouteUsage()
   Custom Base Color=NamedRerouteUsage()
   Custom Base Color Blend=NamedRerouteUsage()
   Primary Color=NamedRerouteUsage()
   Primary Color Blend=NamedRerouteUsage()
   Secondary Color=NamedRerouteUsage()
   Secondary Color Blend=NamedRerouteUsage()
   Tertiary Color=NamedRerouteUsage()
   Tertiary Blend Color=NamedRerouteUsage()
   Pattern Color=NamedRerouteUsage()
  ) {materialFunction=&/MF_SK_Anime_Custom_Color}
  n3 = StaticSwitchParameter(
   True=NamedRerouteUsage()
   False=VectorParameter() {group=01_Anime Maps, parameterName=Color_Lit, sortPriority=10, defaultValue=<rgba 0.9 0.9 0.9 0>}.RGB
  ) {group=01_Anime Maps, parameterName=UseColorLitMap, sortPriority=0}
  n4 = NamedRerouteUsage()
  NamedRerouteDeclaration(None=MaterialFunctionCall!(LitBaseColorIn=StaticSwitch(True=@n2.Color, False=@n3, Value=@n4)) {materialFunction=/Game/Packages/Fortress_SharedMaterials/Cinematics/Anime/MF_CINE_LitDiffuseTint}.Result) {name=Color_Lit, nodeColor=<rgba 0 0.4609 1 1>}
  NamedRerouteDeclaration(None=FunctionInput(Preview=TextureCoordinate()) {bUsePreviewValueAsDefault=true, inputName=CustomUV, inputType=FunctionInput_Vector2}) {name=UV Coords, nodeColor=<rgba 0.5 0.5 0 1>}
  n5 = NamedRerouteUsage()
  NamedRerouteDeclaration(None=ChannelMaskParameter(None=@n5) {group=Color Customization, parameterName=BaseColorAdvancedBlendChannel, sortPriority=18}.R) {name=AdvancedBlendMask, nodeColor=<rgba 0.5703 1 0 1>}
  n6 = ChannelMaskParameter(None=@n5) {group=Color Customization, maskChannel=Green, parameterName=BaseColorValueChannel, sortPriority=15, defaultValue=<rgba 0 1 0 0>}
  n7 = Add(A=@n6.R, B=OneMinus(None=ScalarParameter() {group=Color Customization, parameterName=Global_BaseColorValueBias, sortPriority=16}))
  NamedRerouteDeclaration(None=Power(
    Base=Saturate(None=@n7)
    Exp=ScalarParameter() {defaultValue=1, group=Color Customization, parameterName=Global_BaseColorValuePow, sortPriority=17}
   )) {name=BaseColorValue, nodeColor=<rgba 1 0 0.8906 1>}
  n8 = TextureSampleParameter2D(UVs=NamedRerouteUsage()) {
   desc="R=Spec\r\nG=Rough\r\nB=Metallic"
   group=01_Anime Maps
   parameterName=SRM_Map
   samplerType=SAMPLERTYPE_Masks
   sortPriority=15
   texture=&T_Fortnite_Default_S
   channelNames {a="", b=Metallic, g=Roughness, r=Specular}
  }
  NamedRerouteDeclaration(None=@n8.RGB) {name=SRM_Map, nodeColor=<rgba 0.9219 0 1 1>}
  n9 = TextureSampleParameter2D(UVs=NamedRerouteUsage()) {group=01_Anime Maps, parameterName=Color_Shaded_Map, sortPriority=11, texture=/Game/Global/Textures/Default/Blanks/T_MidgreysRGB}
  NamedRerouteDeclaration(None=@n9.RGB) {name=Color_Shaded_Map, nodeColor=<rgba 0 1 0.6016 1>}
  n10 = TextureSampleParameter2D(UVs=NamedRerouteUsage()) {group=01_Anime Maps, parameterName=DFL_Color_Map, sortPriority=18, texture=&T_BlacksRGB}
  NamedRerouteDeclaration(None=@n10.RGB) {name=DFL_Color_Map, nodeColor=<rgba 1 0.2812 0 1>}
  n11 = TextureSampleParameter2D() {
   group=Color Customization
   parameterName=Pattern Mask
   samplerSource=SSM_Wrap_WorldGroupSettings
   samplerType=SAMPLERTYPE_Masks
   sortPriority=51
   texture=&T_BlackMask
  }
  n12 = SmoothStep(
   Min=ScalarParameter() {defaultValue=0.45, group=Color Customization, parameterName=PatternMask_SmoothStepMin, sortPriority=52}
   Max=ScalarParameter() {defaultValue=0.55, group=Color Customization, parameterName=PatternMask_SmoothStepMax, sortPriority=52}
   Value=@n11.RGB
  ) {constMin=0.5}
  n13 = VectorParameter() {
   desc=Put 1.0 in the channel that that the pattern mask is in and 0.0 in the other channels
   group=Color Customization
   parameterName=Pattern Mask Channel
   sortPriority=52
   defaultValue=<rgba 1 0 0 1>
  }
  NamedRerouteDeclaration(None=Saturate(None=DotProduct(A=@n12, B=@n13.RGB))) {name=PatternMask, nodeColor=<rgba 0.1719 1 0 1>}
  n14 = TextureSampleParameter2D(UVs=NamedRerouteUsage()) {
   group=01_Anime Maps
   parameterName=DFL_Map
   samplerType=SAMPLERTYPE_Alpha
   sortPriority=14
   texture=/Game/Global/Textures/Default/Blanks/T_BlackAlpha
  }
  NamedRerouteDeclaration(None=@n14.RGB) {name=DFL_Map, nodeColor=<rgba 0 0.0391 1 1>}
  n15 = ScalarParameter() {defaultValue=0.45, group=Color Customization, parameterName=TechArtMask_SmoothStepMin, sortPriority=-10}
  n16 = ScalarParameter() {defaultValue=0.55, group=Color Customization, parameterName=TechArtMask_SmoothStepMax, sortPriority=-10}
  n17 = NamedRerouteUsage()
  NamedRerouteDeclaration(None=ChannelMaskParameter(None=SmoothStep(Min=@n15, Max=@n16, Value=@n17) {constMax=0.55, constMin=0.5}) {group=Color Customization, parameterName=Tech Art Mask Channel, sortPriority=-10}.R) {name=Mask, nodeColor=<rgba 0 0.8125 1 1>}
  n18 = TextureSampleParameter2D(UVs=NamedRerouteUsage()) {group=01_Anime Maps, parameterName=SpecularColor_Map, sortPriority=19, texture=&T_BlacksRGB}
  NamedRerouteDeclaration(None=@n18.RGB) {name=SpecularColor_Map, nodeColor=<rgba 0.3594 1 0 1>}
  n19 = TextureSampleParameter2D(UVs=NamedRerouteUsage()) {
   group=01_Anime Maps
   parameterName=Normals
   samplerType=SAMPLERTYPE_Normal
   sortPriority=13
   texture=&FlatNormalBlue
  }
  NamedRerouteDeclaration(None=@n19.RGB) {name=Normals, nodeColor=<rgba 1 0 0.6797 1>}
  n20 = TextureSampleParameter2D() {
   group=Color Customization
   parameterName=CustomColorMask
   samplerSource=SSM_Wrap_WorldGroupSettings
   samplerType=SAMPLERTYPE_Masks
   sortPriority=-6
   texture=&T_BlackMask
  }
  n21 = SmoothStep(
   Min=ScalarParameter() {defaultValue=0.45, group=Color Customization, parameterName=CustomColorMask_SmoothStepMin, sortPriority=-6}
   Max=ScalarParameter() {defaultValue=0.55, group=Color Customization, parameterName=CustomColorMask_SmoothStepMax, sortPriority=-6}
   Value=@n20.RGB
  ) {constMin=0.5}
  NamedRerouteDeclaration(None=ChannelMaskParameter(None=@n21) {group=Color Customization, parameterName=PrimaryMaskChannel, sortPriority=-5}.R) {name=PrimaryMask, nodeColor=<rgba 1 0.0596 0.1687 1>}
  n22 = TextureSampleParameter2D(UVs=NamedRerouteUsage()) {group=01_Anime Maps, parameterName=Emissive_Map, sortPriority=17, texture=&T_BlacksRGB}
  NamedRerouteDeclaration(None=@n22.RGB) {name=Emissive_Map, nodeColor=<rgba 0 1 1 1>}
  n23 = VectorParameter() {group=Color Customization, parameterName=Lit_BaseColor, sortPriority=30, defaultValue=<rgba 1 1 1 1>}
  NamedRerouteDeclaration(None=@n23.RGB) {name=Lit_BaseColor, nodeColor=<rgba 0 0.5312 1 1>}
  n24 = VectorParameter() {group=Color Customization, parameterName=Lit_BaseBlendColor, sortPriority=31, defaultValue=<rgba 1 1 1 1>}
  NamedRerouteDeclaration(None=@n24.RGB) {name=Lit_BaseBlendColor, nodeColor=<rgba 0 0.5312 1 1>}
  n25 = VectorParameter() {group=Color Customization, parameterName=Shaded_BaseColor, sortPriority=30, defaultValue=<rgba 1 1 1 1>}
  NamedRerouteDeclaration(None=@n25.RGB) {name=Shaded_BaseColor, nodeColor=<rgba 0 0.5312 1 1>}
  n26 = VectorParameter() {group=Color Customization, parameterName=Shaded_BaseBlendColor, sortPriority=31, defaultValue=<rgba 1 1 1 1>}
  NamedRerouteDeclaration(None=@n26.RGB) {name=Shaded_BaseBlendColor, nodeColor=<rgba 0 0.5312 1 1>}
  NamedRerouteDeclaration(None=VectorParameter() {group=Color Customization, parameterName=Lit_PrimaryColor, defaultValue=<rgba 1 1 1 1>}.RGB) {name=Lit_PrimaryColor, nodeColor=<rgba 1 0.1875 0 1>}
  n27 = VectorParameter() {group=Color Customization, parameterName=Lit_PrimaryBlendColor, sortPriority=33, defaultValue=<rgba 1 1 1 1>}
  NamedRerouteDeclaration(None=@n27.RGB) {name=Lit_PrimaryBlendColor, nodeColor=<rgba 1 0.1875 0 1>}
  NamedRerouteDeclaration(None=VectorParameter() {group=Color Customization, parameterName=Shaded_PrimaryColor, defaultValue=<rgba 1 1 1 1>}.RGB) {name=Shaded_PrimaryColor, nodeColor=<rgba 1 0.1875 0 1>}
  n28 = VectorParameter() {group=Color Customization, parameterName=Shaded_PrimaryBlendColor, sortPriority=33, defaultValue=<rgba 1 1 1 1>}
  NamedRerouteDeclaration(None=@n28.RGB) {name=Shaded_PrimaryBlendColor, nodeColor=<rgba 1 0.1875 0 1>}
  n29 = VectorParameter() {group=Color Customization, parameterName=Lit_SecondaryColor, sortPriority=34, defaultValue=<rgba 1 1 1 1>}
  NamedRerouteDeclaration(None=@n29.RGB) {name=Lit_SecondaryColor, nodeColor=<rgba 0 1 0.9063 1>}
  n30 = ChannelMaskParameter(None=@n21) {group=Color Customization, maskChannel=Green, parameterName=SecondaryMaskChannel, sortPriority=-4, defaultValue=<rgba 0 1 0 0>}
  NamedRerouteDeclaration(None=@n30.R) {name=SecondaryMask, nodeColor=<rgba 0 1 0.0283 1>}
  n31 = VectorParameter() {group=Color Customization, parameterName=Lit_SecondaryBlendColor, sortPriority=35, defaultValue=<rgba 1 1 1 1>}
  NamedRerouteDeclaration(None=@n31.RGB) {name=Lit_SecondaryBlendColor, nodeColor=<rgba 0 1 0.9062 1>}
  n32 = VectorParameter() {group=Color Customization, parameterName=Shaded_SecondaryColor, sortPriority=34, defaultValue=<rgba 1 1 1 1>}
  NamedRerouteDeclaration(None=@n32.RGB) {name=Shaded_SecondaryColor, nodeColor=<rgba 0 1 0.9062 1>}
  n33 = VectorParameter() {group=Color Customization, parameterName=Shaded_SecondaryBlendColor, sortPriority=35, defaultValue=<rgba 1 1 1 1>}
  NamedRerouteDeclaration(None=@n33.RGB) {name=Shaded_SecondaryBlendColor, nodeColor=<rgba 0 1 0.9062 1>}
  n34 = VectorParameter() {group=Color Customization, parameterName=Lit_TertiaryColor, sortPriority=36, defaultValue=<rgba 1 1 1 1>}
  NamedRerouteDeclaration(None=@n34.RGB) {name=Lit_TertiaryColor, nodeColor=<rgba 1 0 0.375 1>}
  n35 = VectorParameter() {group=Color Customization, parameterName=Lit_TertiaryBlendColor, sortPriority=37, defaultValue=<rgba 1 1 1 1>}
  NamedRerouteDeclaration(None=@n35.RGB) {name=Lit_TertiaryBlendColor, nodeColor=<rgba 1 0 0.375 1>}
  n36 = VectorParameter() {group=Color Customization, parameterName=Shaded_TertiaryColor, sortPriority=36, defaultValue=<rgba 1 1 1 1>}
  NamedRerouteDeclaration(None=@n36.RGB) {name=Shaded_TertiaryColor, nodeColor=<rgba 1 0 0.375 1>}
  n37 = VectorParameter() {group=Color Customization, parameterName=Shaded_TertiaryBlendColor, sortPriority=37, defaultValue=<rgba 1 1 1 1>}
  NamedRerouteDeclaration(None=@n37.RGB) {name=Shaded_TertiaryBlendColor, nodeColor=<rgba 1 0 0.375 1>}
  n38 = VectorParameter() {group=Color Customization, parameterName=Lit_PatternColor, sortPriority=53, defaultValue=<rgba 1 1 1 1>}
  NamedRerouteDeclaration(None=@n38.RGB) {name=Lit_PatternColor, nodeColor=<rgba 0 1 0.3438 1>}
  n39 = VectorParameter() {group=Color Customization, parameterName=Shaded_PatternColor, sortPriority=53, defaultValue=<rgba 1 1 1 1>}
  NamedRerouteDeclaration(None=@n39.RGB) {name=Shaded_PatternColor, nodeColor=<rgba 0 1 0.3438 1>}
  n40 = ChannelMaskParameter(None=@n21) {group=Color Customization, maskChannel=Blue, parameterName=TertiaryMaskChannel, sortPriority=-3, defaultValue=<rgba 0 0 1 0>}
  NamedRerouteDeclaration(None=@n40.R) {name=TertiaryMask, nodeColor=<rgba 0 0.2243 1 1>}
  n41 = NamedRerouteUsage()
  n42 = ChannelMaskParameter(None=@n41) {group=01_Anime Maps, parameterName=STT_Channel_Sheen, sortPriority=21}
  n43 = ChannelMaskParameter(None=@n41) {group=01_Anime Maps, maskChannel=Green, parameterName=STT_Channel_Threshold, sortPriority=22, defaultValue=<rgba 0 1 0 0>}
  n44 = ChannelMaskParameter(None=@n41) {group=01_Anime Maps, maskChannel=Blue, parameterName=STT_Channel_Thickness, sortPriority=23, defaultValue=<rgba 0 0 1 0>}
  n45 = ScalarParameter() {group=01_Anime Maps, parameterName=STT_Sheen, sliderMax=1, sortPriority=21}
  n46 = ScalarParameter() {defaultValue=0.5, group=01_Anime Maps, parameterName=STT_Threshold, sliderMax=1, sortPriority=22}
  n47 = ScalarParameter() {defaultValue=1, group=01_Anime Maps, parameterName=STT_Thickness, sliderMax=1, sortPriority=23}
  NamedRerouteDeclaration(None=StaticSwitchParameter(
    True=MaterialFunctionCall(X=@n42.R, Y=@n43.R, Z=@n44.R) {materialFunction=&MakeFloat3}.Result
    False=MaterialFunctionCall(X=@n45, Y=@n46, Z=@n47) {materialFunction=&MakeFloat3}.Result
   ) {group=01_Anime Maps, parameterName=UseSTTMap, sortPriority=3}) {name=STT, nodeColor=<rgba 0 0.5078 1 1>}
  n48 = Multiply(
   A=NamedRerouteUsage()
   B=ScalarParameter() {defaultValue=0.5, group=01_Anime Maps, parameterName=Dynamic_ColorShaded_Darkening, sliderMax=1, sortPriority=11}
  )
  n49 = StaticSwitchParameter(
   True=MaterialFunctionCall(
    Tech Art Mask=NamedRerouteUsage()
    Primary Mask=NamedRerouteUsage()
    Secondary Mask=NamedRerouteUsage()
    Tertiary Mask=NamedRerouteUsage()
    Pattern Mask=NamedRerouteUsage()
    Advanced Blend Mask=NamedRerouteUsage()
    Use Advanced Color Blend=NamedRerouteUsage()
    Diffuse Base Color=NamedRerouteUsage()
    Diffuse Base Color Value=NamedRerouteUsage()
    Custom Base Color=NamedRerouteUsage()
    Custom Base Color Blend=NamedRerouteUsage()
    Primary Color=NamedRerouteUsage()
    Primary Color Blend=NamedRerouteUsage()
    Secondary Color=NamedRerouteUsage()
    Secondary Color Blend=NamedRerouteUsage()
    Tertiary Color=NamedRerouteUsage()
    Tertiary Blend Color=NamedRerouteUsage()
    Pattern Color=NamedRerouteUsage()
   ) {materialFunction=&MF_SK_Anime_Custom_Color}.Color
   False=@n48
  ) {group=Color Customization, parameterName=UseShadedCustomColors, sortPriority=2}
  n50 = VectorParameter() {group=01_Anime Maps, parameterName=Color_Shaded, sortPriority=11, defaultValue=<rgba 0.18 0.18 0.18 0>}
  n51 = StaticSwitchParameter(
   True=StaticSwitchParameter(True=NamedRerouteUsage(), False=@n48) {group=01_Anime Maps, parameterName=UseColorShadedMap, sortPriority=2}
   False=@n50.RGB
  ) {group=01_Anime Maps, parameterName=UseDynamicColorShaded, sortPriority=1}
  n52 = NamedRerouteUsage()
  NamedRerouteDeclaration(None=MaterialFunctionCall!(ShadedBaseColorIn=StaticSwitch(True=@n49, False=@n51, Value=@n52)) {materialFunction=/Game/Packages/Fortress_SharedMaterials/Cinematics/Anime/MF_CINE_ShadedDiffuseTint}.Result) {name=Color_Shaded, nodeColor=<rgba 1 0.3984 0 1>}
  n53 = ScalarParameter() {defaultValue=0.5, group=01_Anime Maps, parameterName=SRM_Specular, sliderMax=1, sortPriority=15}
  n54 = ScalarParameter() {defaultValue=0.5, group=01_Anime Maps, parameterName=SRM_Roughness, sliderMax=1, sortPriority=15}
  n55 = ScalarParameter() {group=01_Anime Maps, parameterName=SRM_Metallic, sliderMax=1, sortPriority=15}
  NamedRerouteDeclaration(None=StaticSwitchParameter(
    True=NamedRerouteUsage()
    False=MaterialFunctionCall(X=@n53, Y=@n54, Z=@n55) {materialFunction=&MakeFloat3}.Result
   ) {group=01_Anime Maps, parameterName=UseSRMMap, sortPriority=5}) {name=SRM, nodeColor=<rgba 0.8281 1 0 1>}
  n56 = Multiply(A=NamedRerouteUsage(), B=ScalarParameter() {defaultValue=1, group=02_Anime Parameters, parameterName=Emissive_Intensity, sliderMax=10})
  n57 = CameraVectorWS()
  n58 = PixelNormalWS()
  n59 = Saturate(None=Power(Base=DotProduct(A=@n57, B=@n58), Exp=ScalarParameter() {group=02_Anime Parameters, parameterName=Emissive_FresnelExponent, sliderMax=8}))
  NamedRerouteDeclaration(None=StaticSwitchParameter(True=Multiply(A=@n56, B=@n59), False=Constant3Vector()) {group=01_Anime Maps, parameterName=UseEmissiveMap, sortPriority=7}) {name=Emissive, nodeColor=<rgba 0.8516 0 1 1>}
  n60 = StaticSwitchParameter(
   True=NamedRerouteUsage()
   False=Multiply(A=NamedRerouteUsage(), B=ScalarParameter() {group=01_Anime Maps, parameterName=Dynamic_DFLColor_Darkening, sliderMax=1, sortPriority=19})
  ) {group=01_Anime Maps, parameterName=UseDFL_ColorMap, sortPriority=9}
  n61 = VectorParameter() {group=01_Anime Maps, parameterName=DFLine_Color, sortPriority=20, defaultValue=<rgba 0 0 0 1>}
  NamedRerouteDeclaration(None=Saturate(None=StaticSwitchParameter(True=@n60, False=@n61.RGB) {group=01_Anime Maps, parameterName=UseDynamicDFLColor, sortPriority=8})) {name=DFL_Color, nodeColor=<rgba 0 1 0.5313 1>}
  n62 = NamedRerouteUsage()
  n63 = ScalarParameter() {defaultValue=1, group=01_Anime Maps, parameterName=SpecularColor_Opacity, sliderMax=1, sortPriority=19}
  n64 = VectorParameter() {
   desc="RGB=Color, A=Opacity"
   group=01_Anime Maps
   parameterName=Specular_Color
   sortPriority=21
   channelNames {a=Opacity, b=Blue Channel, g=Green Channel, r=Red Channel}
   defaultValue=<rgba 1 1 1 1>
  }
  NamedRerouteDeclaration(None=StaticSwitchParameter(True=AppendVector(A=@n62, B=@n63), False=AppendVector(A=@n64, B=@n64)) {group=01_Anime Maps, parameterName=UseSpecularColorMap, sortPriority=9}) {name=Specular, nodeColor=<rgba 1 0.2109 0 1>}
  n65 = TextureSampleParameter2D(UVs=NamedRerouteUsage()) {group=01_Anime Maps, parameterName=Color_Lit_Map, sortPriority=10, texture=&T_WhitesRGB}
  NamedRerouteDeclaration(None=@n65.RGB) {name=Color_Lit_Map, nodeColor=<rgba 1 0 0.1172 1>}
  NamedRerouteDeclaration(None=StaticBoolParameter() {group=Color Customization, parameterName=UseAdvancedColorBlend, sortPriority=17}) {name=AdvancedColorBlend, nodeColor=<rgba 0.7109 0 1 1>}
  n66 = TextureSampleParameter2D(UVs=NamedRerouteUsage()) {
   desc="R=Sheen \r\nG=Threshold\r\nB=Thickness"
   group=01_Anime Maps
   parameterName=STT_Map
   samplerType=SAMPLERTYPE_Masks
   sortPriority=12
   texture=/Game/Global/Textures/Default/Blanks/T_FN_Anime_STT
   channelNames {a="", b=Thickness, g="Threshold ", r=Sheen}
  }
  NamedRerouteDeclaration(None=@n66.RGB) {name=STT_Map, nodeColor=<rgba 0 0.4375 1 1>}
  n67 = TextureSampleParameter2D(UVs=NamedRerouteUsage()) {
   group=01_Anime Maps
   parameterName=FX_Map
   samplerType=SAMPLERTYPE_Masks
   sortPriority=16
   texture=&T_BlackMask
  }
  NamedRerouteDeclaration(None=@n67.RGB) {name=FX_Map, nodeColor=<rgba 0.7578 1 0 1>}
  NamedRerouteDeclaration(None=StaticBoolParameter() {defaultValue=true, group=Color Customization, parameterName=UseCustomColor, sortPriority=-50}) {name=UseCustomColor, nodeColor=<rgba 1 0 0.1875 1>}
 ]}
